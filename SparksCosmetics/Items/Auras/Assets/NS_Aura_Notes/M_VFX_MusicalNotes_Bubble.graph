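# Translucent unlit Niagara sprite material for bubble musical-note particles in the Sparks Notes aura.
# Dynamic parameters drive a sine/cosine UV scale wobble, radial sphere-mask UV distortion for growth, and noise-based disintegration of opacity.
# The distorted UVs sample a mask texture whose red channel shapes opacity and blue channel a highlight, switched by MaskForBubble.
# The highlight adds to particle color, which feeds an emissive adjustment function.
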Material /SparksCosmetics/Items/Auras/Assets/NS_Aura_Notes/M_VFX_MusicalNotes_Bubble
graph {
 properties {bUsedWithNiagaraSprites=true, blendMode=BLEND_Translucent, pixelDepthOffsetMode=PDOM_Legacy, shadingModel=MSM_Unlit, usedShadingModels=Unlit}
 nodes [
  n1 = NamedRerouteUsage()
  n2 = ParticleColor()
  Add(A=@n1, B=@n2.A)
  MaterialFunctionCall(Emissive=Add(A=NamedRerouteUsage(), B=@n2.RGB)) {materialFunction=&MF_FX_EmissiveAdjustments}
  Multiply(A=@n2.A, B=@n1)
  n3 = Subtract(
   A=Add(A=TextureSample() {texture=/Game/Effects/Fort_Effects/Textures/Tile/T_GaussianNoise_01_M}.R)
   B=Multiply(A=DynamicParameter() {defaultValue=<rgba 0 0 0 1>, paramNames [Wobble,Disintegrate,Grow,-]}.Disintegrate) {constB=2}
  )
  n4 = TextureCoordinate()
  n5 = Multiply(A=Time()) {constB=5}
  n6 = Sine(None=@n5)
  n7 = Cosine(None=@n5)
  n8 = ScalarParameter() {defaultValue=0.7, parameterName=Bubble Scale}
  n9 = DynamicParameter() {defaultValue=<rgba 0 0 0 1>, paramNames [Wobble,Disintegrate,Grow,-]}
  n10 = Add(A=Multiply(A=AppendVector(A=@n6, B=@n7), B=Multiply(A=@n8, B=@n9.Wobble)) {constB=0.1})
  n11 = AppendVector(A=Subtract(A=TextureCoordinate()) {constB=0.5}, B=Constant())
  n12 = Constant() {r=0.5}
  n13 = SphereMask(A=TextureCoordinate(), B=@n12, Radius=@n12) {hardnessPercent=0}
  n14 = Multiply(
   A=ComponentMask(None=MaterialFunctionCall(Vector=@n11) {materialFunction=&SafeNormalize}.Result) {g=true, r=true}
   B=Sine(None=Multiply(A=@n13) {constB=0.5})
  )
  n15 = Multiply(
   A=DynamicParameter() {defaultValue=<rgba 0 0 0 1>, paramNames [Wobble,Disintegrate,Grow,-]}.Grow
   B=ScalarParameter() {defaultValue=-0.2, parameterName=Max Grow}
  )
  n16 = TextureSampleParameter2D(UVs=Add(
    A=MaterialFunctionCall(UVs=@n4, Texture Scale=@n10) {materialFunction=&ScaleUVsByCenter}.UVs
    B=Multiply(A=@n14, B=@n15) {constB=0.1}
   )) {parameterName=Mask, samplerType=SAMPLERTYPE_LinearColor, texture=&/T_MusicalNotes_Rounded_M}
  n17 = Step(X=ParticleRandom()) {constY=0.5}
  n18 = LinearInterpolate(A=@n16.R, B=@n16.R, Alpha=@n17)
  NamedRerouteDeclaration(None=Multiply(A=Saturate(None=@n3), B=StaticSwitchParameter(True=@n16.R, False=@n18) {parameterName=MaskForBubble})) {name=Opacity, nodeColor=<rgba 0.3353 0.6433 1 1>}
  NamedRerouteDeclaration(None=Add(A=ParticleColor().RGB, B=NamedRerouteUsage())) {name=Base Color, nodeColor=<rgba 0.2217 1 0.1178 1>}
  n19 = LinearInterpolate(A=@n16.B, B=@n16.B, Alpha=@n17)
  NamedRerouteDeclaration(None=StaticSwitchParameter(True=@n16.B, False=Power(Base=@n19) {constExponent=10}) {parameterName=MaskForBubble}) {name=Highlight, nodeColor=<rgba 1 0.594 0.0552 1>}
  NamedRerouteUsage()
  ScalarParameter() {defaultValue=0.1, parameterName=Roughness}
  TextureCoordinate()
 ]
}
