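# Niagara music-note aura cosmetic for Sparks, attaching sprite glows, notes, bubbles and sparkles to five bones or sockets of a character's skeletal mesh.
# The Aura emitter bursts five looping glow sprites, one per point, with pulsing size and magenta-to-teal color cycling.
# Notes, bubbles and sparkles spawn continuously around those points with curl noise, and notes and bubbles pulse on an audio beat parameter.
# User.IsLargeBody adjusts sizes, radii and offsets.
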NiagaraSystem /SparksCosmetics/Items/Auras/Assets/NS_Aura_Notes/NS_Aura_Notes
dependencies {
 usedDataInterfaces [
  &NiagaraDataInterfaceArrayColor
  &NiagaraDataInterfaceArrayFloat
  &NiagaraDataInterfaceArrayFloat2
  &NiagaraDataInterfaceArrayInt32
  &NiagaraDataInterfaceColorCurve
  &NiagaraDataInterfaceCurve
  &NiagaraDataInterfaceArrayFloat3
  &NiagaraDataInterfaceSkeletalMesh
 ]
 usedDynamicInputs [&MakeVector]
 usedModules [
  &self:SystemSpawnScript.SystemScriptSource.SystemScriptGraph.NiagaraNodeAssignment_0.SetVariables_#1
  &SystemState
 ]
}
emitters {
 Aura {
  emitterUpdateScript [
   &EmitterState(
    Inactive Response=::NewEnumerator1
    Life Cycle Mode=::NewEnumerator1
    Loop Behavior=::NewEnumerator1
    Loop Duration Mode=::NewEnumerator1
    UseLoopDelay=false
   )
   &SpawnBurst_Instantaneous(Spawn Count=5, Use Spawn Probability=false, UseLoopCountLimit=false)
  ]
  particleSpawnScript [
   &PartitionParticles(Number Of Partitions=5)
   &InitializeParticle(
    Color=&SelectLinearColorFromArray(
     Array Sampling Mode=::NewEnumerator0
     Direct Array Index=?
     Color Selection Array=DataInterface() {propertyValues {ColorData [<rgba 1 0 0.8984 1>, <rgba 0 5 3.3421 1>]}}
     Random Seed=0
    )
    Lifetime=&RandomRangeFloat(Minimum=1, Maximum=2)
    Sprite Size Mode=::NewEnumerator3
    Uniform Sprite Size=&Multiply_Float(
     A=&SelectFloatFromArray(
      Direct Array Index=$Particles.PartitionParticles.Partition
      Float Selection Array=DataInterface() {propertyValues {FloatData [60,30,30,40,40]}}
     )
     B=&MakeCustomFloatFromBool(Bool=$User.IsLargeBody, True Float=2, False Float=1)
    )
    UsePositionOffset=false
    Write Material Random=true
   )
   &CameraOffset(Camera Offset Amount=&SelectFloatFromArray(
     Direct Array Index=$Particles.PartitionParticles.Partition
     Float Selection Array=DataInterface() {propertyValues {FloatData [10,10,10,5,5]}}
    ))
   /SetVariables(Particles.GlowCameraOffset=-30) {isSetParameters=true}
   /SetVariables(Particles.GlowScale:Vector2f=&Multiply_Vector2D(
     A=&SelectVector2DFromArray(
      Direct Array Index=$Particles.PartitionParticles.Partition
      Vector 2D Selection Array=DataInterface() {propertyValues {FloatData [<0 0>, <100 100>, <100 100>, <100 100>, <100 100>]}}
     )
     B=&MakeCustomVector2DFromBool(Bool=$User.IsLargeBody, True Vector2D=<1.25 1.25>, False Vector2D=<1 1>)
    )) {isSetParameters=true}
   /SetVariables(Particles.GlowColor=&SelectLinearColorFromArray(
     Array Sampling Mode=::NewEnumerator0
     Direct Array Index=?
     Color Selection Array=DataInterface() {propertyValues {ColorData [<rgba 1 0 0.8984 1>, <rgba 0 1 0.6684 1>]}}
     Random Seed=0
    )) {isSetParameters=true}
  ]
  particleUpdateScript [
   &ParticleState(
    Kill Particles When Lifetime Has Elapsed=false
    Let Infinitely Lived Particles Die When Emitter Deactivates=true
    Loop Particles Lifetime=true
   )
   &SpriteRotationRate(Rotation Rate=25) {enabled=false}
   &SkeletalMeshLocation(
    Bone / Socket Index=&SelectIntFromArray(Direct Array Index=$Particles.PartitionParticles.Partition, Int Selection Array=DataInterface() {propertyValues {IntData [0,1,2,3,4]}})
    Bone Sampling Mode=::NewEnumerator4
    MaskSpawnGroup=false
    Orientation Sampling=::NewEnumerator1
    Skeletal Mesh=$System.Skeletal Mesh
   )
   &SolveForcesAndVelocity(Clamp Velocity=false, Limit Acceleration=false)
   /SetVariables(Particles.GlowColor=&MakeCustomLinearColorFromBool(
     Bool=&RandomBool()
     True Color=&ColorFromCurve(ColorCurve=DataInterface() {propertyValues {
        AlphaCurve {keys [<at 0 1 RCIM_Cubic>,<at 0.3473 0.7>,<at 0.6711 0.5>,<at 1 1>]}
        BlueCurve {keys [<at 0 0.8984 RCIM_Cubic>,<at 0.3745 0.6684>,<at 0.5653 0.6684 RCIM_Cubic>,<at 1 0.8984>]}
        GreenCurve {keys [<at 0 0 RCIM_Cubic>,<at 0.3745 1>,<at 0.5653 1 RCIM_Cubic>,<at 1 0>]}
        RedCurve {keys [<at 0 1 RCIM_Cubic>,<at 0.3745 0>,<at 0.5653 0 RCIM_Cubic>,<at 1 1>]}
       }})
     False Color=&ColorFromCurve(ColorCurve=DataInterface() {propertyValues {
        AlphaCurve {keys [<at 0 1 RCIM_Cubic>,<at 0.3379 1>,<at 0.7546 0.7>,<at 1 1>]}
        BlueCurve {keys [
          <at 0 0.6684 RCIM_Cubic>
          <at 0.3745 0.8984>
          {arriveTangent=-0.3676, interpMode=RCIM_Cubic, leaveTangent=-0.3676, time=0.5653, value=0.8984}
          <at 1 0.6684>
         ]}
        GreenCurve {keys [<at 0 1 RCIM_Cubic>, <at 0.3745 0>, {arriveTangent=1.5987, interpMode=RCIM_Cubic, leaveTangent=1.5987, time=0.5653, value=0}, <at 1 1>]}
        RedCurve {keys [<at 0 0 RCIM_Cubic>, <at 0.3745 1>, {arriveTangent=-1.5987, interpMode=RCIM_Cubic, leaveTangent=-1.5987, time=0.5653, value=1}, <at 1 0>]}
       }})
    )) {isSetParameters=true}
   /SetVariables(Particles.GlowScale:Vector2f=&Multiply_Vector2D_ByFloat(
     Float=&FloatFromCurve(
      FloatCurve=DataInterface() {propertyValues {Curve {keys [
          {interpMode=RCIM_Cubic, tangentMode=RCTM_User, time=0, value=0.575}
          {arriveTangent=0.0501, interpMode=RCIM_Cubic, leaveTangent=-0.2407, tangentMode=RCTM_Break, time=0.4, value=1}
          {arriveTangent=-0.0395, interpMode=RCIM_Cubic, leaveTangent=-2.8523, tangentMode=RCTM_Break, time=0.6, value=1}
          {tangentMode=RCTM_User, time=1, value=0.575}
         ]}}}
      CurveIndex=&Multiply_Float(A=$Particles.NormalizedAge, B=2)
     )
     Vector2D=$Particles.Initial.GlowScale
    )) {isSetParameters=true}
   &Color(Color=&MultiplyLinearColor(Linear Color=$Particles.GlowColor, Scale Factor=&MakeVector4_FromFloat(Float=5)))
  ]
  rendererClasses [&NiagaraSpriteRendererProperties]
  renderers [{rendererClass=&NiagaraSpriteRendererProperties}]
 }
 Bubbles001 {
  emitterUpdateScript [
   &EmitterState(
    Inactive Response=::NewEnumerator1
    Life Cycle Mode=::NewEnumerator1
    Loop Behavior=::NewEnumerator1
    Loop Duration Mode=::NewEnumerator1
    UseLoopDelay=false
   )
   &SpawnRate(SpawnRate=3, Use Spawn Probability=false)
  ]
  particleSpawnScript [
   &InitializeParticle(
    Color=&SelectLinearColorFromArray(
     Array Sampling Mode=::NewEnumerator0
     Direct Array Index=?
     Color Selection Array=DataInterface() {propertyValues {ColorData [<rgba 10 0 8.9836 1>, <rgba 0 10 6.6841 1>, <rgba 0.6302 0 0.5659 1>, <rgba 0 0.6615 0.441 1>]}}
     Random Seed=0
    )
    Lifetime Min=1.5
    Lifetime Mode=::NewEnumerator1
    Position Mode=::NewEnumerator1
    Sprite Size Mode=::NewEnumerator1
    Uniform Sprite Size Min=8
    UsePositionOffset=false
    Write Material Random=true
   ) {module=InitializeParticle001}
   &PartitionParticles(Number Of Partitions=5)
   &ShapeLocation(
    Apply To Particle Position=::NewEnumerator1
    MaskSpawnGroup=false
    Override Seed=false
    Shape Origin=&VectorFromFloat(Value=0)
    Sphere Radius=&MakeCustomFloatFromBool(Bool=$User.IsLargeBody, True Float=15, False Float=10)
    UseOwnerScale=true
   )
   &CameraOffset(Camera Offset Amount=-80)
  ]
  particleUpdateScript [
   &ParticleState()
   &ScaleSpriteSize(
    Scale Sprite Size Mode=::NewEnumerator0
    Uniform Scale Factor=&Lerp_Float(
     A=1
     B=&RandomRangeFloat(Minimum=1.3, Maximum=1.5)
     Alpha=&FloatFromCurve(
      FloatCurve=DataInterface() {propertyValues {Curve {keys [
          {arriveTangent=9.0725, interpMode=RCIM_Cubic, leaveTangent=9.0725, tangentMode=RCTM_User, time=0, value=0}
          {arriveTangent=-2.5249, interpMode=RCIM_Cubic, leaveTangent=-2.5249, tangentMode=RCTM_User, time=0.1408, value=0.9847}
          {arriveTangent=-0.0405, interpMode=RCIM_Cubic, leaveTangent=-0.0405, tangentMode=RCTM_User, time=1, value=0}
         ]}}}
      CurveIndex=$NPC.NPC_Sparks_Audio_Parameters001.BeatLifetime
     )
    )
   )
   &ScaleSpriteSize(Uniform Curve Sprite Scale=DataInterface() {propertyValues {Curve {keys [
        {arriveTangent=24.6732, interpMode=RCIM_Cubic, leaveTangent=24.6732, tangentMode=RCTM_User, time=0, value=0}
        {arriveTangent=-0.0159, interpMode=RCIM_Cubic, leaveTangent=-0.0159, tangentMode=RCTM_User, time=0.05, value=1}
        {arriveTangent=0.0202, interpMode=RCIM_Cubic, leaveTangent=0.0202, tangentMode=RCTM_User, time=0.95, value=1}
        {arriveTangent=27.063, interpMode=RCIM_Cubic, leaveTangent=27.063, tangentMode=RCTM_User, time=1, value=1.75}
       ]}}}) {module=ScaleSpriteSize001}
   &SpriteRotationRate(Rotation Rate=&Multiply_Float(
     A=50
     B=&Sine(
      Normalized Angle=&Add_Float(A=0, B=$Particles.Age)
      Period=&RandomRangeFloat(Minimum=2, Maximum=5)
      Scale=0.5
      Bias=0
     )
    ))
   &SkeletalMeshLocation(
    Bone / Socket Index=$Particles.PartitionParticles.Partition
    Bone Sampling Mode=::NewEnumerator4
    MaskSpawnGroup=false
    Position Sampling=::NewEnumerator1
    Sampled Position Offset=&AddVectorToPosition(
     Position=$StackContext.ShapeLocation.ShapePosition
     Vector=&SelectVectorFromArray(
      Direct Array Index=$Particles.PartitionParticles.Partition
      Vector Selection Array=DataInterface() {propertyValues {FloatData [<0 0 30>, <0 -15 20>, <0 15 20>, <0 -15 20>, <0 15 20>]}}
     )
    )
    Sampled Position Offset Coordinate Space:ENiagaraSkelMeshTransforms=::NewEnumerator1
    Skeletal Mesh=$System.Skeletal Mesh
    Source Movement Scale:Vector3f=<0 0 0>
   ) {module=SkeletalMeshLocation001}
   &CurlNoiseForce(Pan Noise=true, Pan Noise Field=<0.05 0.1 0.2>) {module=CurlNoiseForce001}
   &SolveForcesAndVelocity(Clamp Velocity=false, Limit Acceleration=false)
   &ScaleColor(
    Scale Alpha=&FloatFromCurve(FloatCurve=DataInterface() {propertyValues {Curve {keys [{interpMode=RCIM_Cubic, tangentMode=RCTM_User, time=0, value=1}, {interpMode=RCIM_Cubic, tangentMode=RCTM_User, time=1, value=0}]}}})
    ScaleA=true
    ScaleRGB=true
   )
   &DynamicMaterialParameters(
    Index 0 Param 1=&FloatFromCurve(FloatCurve=DataInterface() {propertyValues {Curve {keys [
         {arriveTangent=-10.8792, interpMode=RCIM_Cubic, leaveTangent=-10.8792, tangentMode=RCTM_User, time=0, value=1}
         {tangentMode=RCTM_User, time=0.17, value=0}
        ]}}})
    Index 0 Param 2=&FloatFromCurve(FloatCurve=DataInterface() {propertyValues {Curve {keys [
         {arriveTangent=0.0202, interpMode=RCIM_Cubic, leaveTangent=0.0202, tangentMode=RCTM_User, time=0.98, value=0}
         {arriveTangent=229.7168, interpMode=RCIM_Cubic, leaveTangent=229.7168, tangentMode=RCTM_User, time=1, value=1.5}
        ]}}})
    Module.Param0WriteEnabled=true
    Module.Param1WriteEnabled=true
    Module.Param2WriteEnabled=true
    Module.Param3WriteEnabled=true
   )
  ]
  rendererClasses [&NiagaraSpriteRendererProperties]
  renderers [{rendererClass=&NiagaraSpriteRendererProperties}]
 }
 Notes001 {
  emitterUpdateScript [
   &EmitterState(
    Inactive Response=::NewEnumerator1
    Life Cycle Mode=::NewEnumerator1
    Loop Behavior=::NewEnumerator1
    Loop Duration Mode=::NewEnumerator1
    UseLoopDelay=false
   )
   &SpawnRate(SpawnRate=2, Use Spawn Probability=false)
  ]
  particleSpawnScript [
   &InitializeParticle(
    Color=&SelectLinearColorFromArray(
     Array Sampling Mode=::NewEnumerator0
     Direct Array Index=?
     Color Selection Array=DataInterface() {propertyValues {ColorData [<rgba 10 0 8.9836 1>, <rgba 0 10 6.6841 1>, <rgba 0.625 0 0.5613 1>, <rgba 0 0.5312 0.3542 1>]}}
     Random Seed=0
    )
    Lifetime Min=1.5
    Lifetime Mode=::NewEnumerator1
    Position Mode=::NewEnumerator1
    Sprite Size Mode=::NewEnumerator3
    Uniform Sprite Size=&MakeCustomFloatFromBool(
     Bool=$User.IsLargeBody
     True Float=&RandomRangeFloat(Minimum=20, Maximum=30)
     False Float=&RandomRangeFloat(Minimum=16, Maximum=20)
    )
    UsePositionOffset=false
    Write Material Random=true
   ) {module=InitializeParticle001}
   &PartitionParticles(Number Of Partitions=5)
   &ShapeLocation(
    Apply To Particle Position=::NewEnumerator1
    MaskSpawnGroup=false
    Override Seed=false
    Shape Origin=&VectorFromFloat(Value=0)
    Sphere Radius=&MakeCustomFloatFromBool(Bool=$User.IsLargeBody, True Float=30, False Float=30)
    UseOwnerScale=true
   )
   &CameraOffset(Camera Offset Amount=-80)
  ]
  particleUpdateScript [
   &ParticleState()
   &ScaleSpriteSize(
    Scale Sprite Size Mode=::NewEnumerator0
    Uniform Scale Factor=&Lerp_Float(
     A=1
     B=&RandomRangeFloat(Minimum=1.3, Maximum=1.5)
     Alpha=&FloatFromCurve(
      FloatCurve=DataInterface() {propertyValues {Curve {keys [
          {arriveTangent=9.0725, interpMode=RCIM_Cubic, leaveTangent=9.0725, tangentMode=RCTM_User, time=0, value=0}
          {arriveTangent=-2.5249, interpMode=RCIM_Cubic, leaveTangent=-2.5249, tangentMode=RCTM_User, time=0.1408, value=0.9847}
          {arriveTangent=-0.0405, interpMode=RCIM_Cubic, leaveTangent=-0.0405, tangentMode=RCTM_User, time=1, value=0}
         ]}}}
      CurveIndex=$NPC.NPC_Sparks_Audio_Parameters001.BeatLifetime
     )
    )
   )
   &ScaleSpriteSize(Uniform Curve Sprite Scale=DataInterface() {propertyValues {Curve {keys [
        {arriveTangent=24.6732, interpMode=RCIM_Cubic, leaveTangent=24.6732, tangentMode=RCTM_User, time=0, value=0}
        {arriveTangent=-0.0159, interpMode=RCIM_Cubic, leaveTangent=-0.0159, tangentMode=RCTM_User, time=0.05, value=1}
        {arriveTangent=0.0202, interpMode=RCIM_Cubic, leaveTangent=0.0202, tangentMode=RCTM_User, time=0.95, value=1}
        {arriveTangent=27.063, interpMode=RCIM_Cubic, leaveTangent=27.063, tangentMode=RCTM_User, time=1, value=1.75}
       ]}}}) {module=ScaleSpriteSize001}
   &SpriteRotationRate(Rotation Rate=&Multiply_Float(
     A=50
     B=&Sine(
      Normalized Angle=&Add_Float(A=0, B=$Particles.Age)
      Period=&RandomRangeFloat(Minimum=2, Maximum=5)
      Scale=0.5
      Bias=0
     )
    ))
   &SkeletalMeshLocation(
    Bone / Socket Index=$Particles.PartitionParticles.Partition
    Bone Sampling Mode=::NewEnumerator4
    MaskSpawnGroup=false
    Position Sampling=::NewEnumerator1
    Sampled Position Offset=&MakeCustomVectorFromBool(
     Bool=$User.IsLargeBody
     True Vector=&AddVectorToPosition(
      Position=$StackContext.ShapeLocation.ShapePosition
      Vector=&SelectVectorFromArray(
       Direct Array Index=$Particles.PartitionParticles.Partition
       Vector Selection Array=DataInterface() {propertyValues {FloatData [<0 0 15>, <0 -15 20>, <0 15 20>, <0 -20 20>, <0 20 20>]}}
      )
     )
     False Vector=&AddVectorToPosition(
      Position=$StackContext.ShapeLocation.ShapePosition
      Vector=&SelectVectorFromArray(
       Direct Array Index=$Particles.PartitionParticles.Partition
       Vector Selection Array=DataInterface() {propertyValues {FloatData [<0 0 15>, <0 -15 20>, <0 15 20>, <0 -10 20>, <0 10 20>]}}
      )
     )
    )
    Sampled Position Offset Coordinate Space:ENiagaraSkelMeshTransforms=::NewEnumerator1
    Skeletal Mesh=$System.Skeletal Mesh
    Source Movement Scale:Vector3f=<0 0 0>
   ) {module=SkeletalMeshLocation001}
   &CurlNoiseForce(Pan Noise=true, Pan Noise Field=<0.05 0.1 0.2>) {module=CurlNoiseForce001}
   &SolveForcesAndVelocity(Clamp Velocity=false, Limit Acceleration=false)
   &ScaleColor(
    Scale Alpha=&FloatFromCurve(FloatCurve=DataInterface() {propertyValues {Curve {keys [{interpMode=RCIM_Cubic, tangentMode=RCTM_User, time=0.95, value=1}, {interpMode=RCIM_Cubic, tangentMode=RCTM_User, time=1, value=0}]}}})
    ScaleA=true
    ScaleRGB=true
   )
   &DynamicMaterialParameters(
    Index 0 Param 1=&FloatFromCurve(FloatCurve=DataInterface() {propertyValues {Curve {keys [
         {arriveTangent=-10.8792, interpMode=RCIM_Cubic, leaveTangent=-10.8792, tangentMode=RCTM_User, time=0, value=1}
         {tangentMode=RCTM_User, time=0.17, value=0}
        ]}}})
    Index 0 Param 2=&FloatFromCurve(FloatCurve=DataInterface() {propertyValues {Curve {keys [
         {arriveTangent=0.0202, interpMode=RCIM_Cubic, leaveTangent=0.0202, tangentMode=RCTM_User, time=0.98, value=0}
         {arriveTangent=229.7168, interpMode=RCIM_Cubic, leaveTangent=229.7168, tangentMode=RCTM_User, time=1, value=1.5}
        ]}}})
    Module.Param0WriteEnabled=true
    Module.Param1WriteEnabled=true
    Module.Param2WriteEnabled=true
    Module.Param3WriteEnabled=true
   )
  ]
  rendererClasses [&NiagaraSpriteRendererProperties]
  renderers [{rendererClass=&NiagaraSpriteRendererProperties}]
 }
 Sparkles {
  emitterUpdateScript [
   &EmitterState(
    Inactive Response=::NewEnumerator1
    Life Cycle Mode=::NewEnumerator1
    Loop Behavior=::NewEnumerator1
    Loop Duration Mode=::NewEnumerator1
    UseLoopDelay=false
   )
   &SpawnRate(SpawnRate=5, Use Spawn Probability=false)
  ]
  particleSpawnScript [
   &InitializeParticle(
    Color=&SelectLinearColorFromArray(
     Array Sampling Mode=::NewEnumerator0
     Direct Array Index=?
     Color Selection Array=DataInterface() {propertyValues {ColorData [<rgba 10 0 8.9836 1>, <rgba 0 10 6.6841 1>, <rgba 0.6302 0 0.5659 1>, <rgba 0 0.6615 0.441 1>]}}
     Random Seed=0
    )
    Lifetime Min=1.5
    Lifetime Mode=::NewEnumerator1
    Position Mode=::NewEnumerator1
    Sprite Size Mode=::NewEnumerator1
    Uniform Sprite Size Max=24
    Uniform Sprite Size Min=16
    UsePositionOffset=false
    Write Material Random=true
   ) {module=InitializeParticle001}
   &PartitionParticles(Number Of Partitions=5)
   &ShapeLocation(
    Apply To Particle Position=::NewEnumerator1
    MaskSpawnGroup=false
    Override Seed=false
    Shape Origin=&VectorFromFloat(Value=0)
    Sphere Radius=&MakeCustomFloatFromBool(Bool=$User.IsLargeBody, True Float=15, False Float=10)
    UseOwnerScale=true
   )
   &CameraOffset(Camera Offset Amount=-80)
  ]
  particleUpdateScript [
   &ParticleState()
   &SpriteRotationRate(Rotation Rate=&RandomRangeFloat(Minimum=-100, Maximum=100)) {module=SpriteRotationRate001}
   &ScaleSpriteSize(Uniform Curve Sprite Scale=DataInterface() {propertyValues {Curve {keys [
        {arriveTangent=9.0725, interpMode=RCIM_Cubic, leaveTangent=9.0725, tangentMode=RCTM_User, time=0, value=0}
        {arriveTangent=-2.5249, interpMode=RCIM_Cubic, leaveTangent=-2.5249, tangentMode=RCTM_User, time=0.1408, value=0.9847}
        {arriveTangent=-0.0405, interpMode=RCIM_Cubic, leaveTangent=-0.0405, tangentMode=RCTM_User, time=1, value=0}
       ]}}}) {module=ScaleSpriteSize001}
   &SpriteRotationRate(Rotation Rate=&Multiply_Float(
     A=50
     B=&Sine(
      Normalized Angle=&Add_Float(A=0, B=$Particles.Age)
      Period=&RandomRangeFloat(Minimum=2, Maximum=5)
      Scale=0.5
      Bias=0
     )
    ))
   &SkeletalMeshLocation(
    Bone / Socket Index=$Particles.PartitionParticles.Partition
    Bone Sampling Mode=::NewEnumerator4
    MaskSpawnGroup=false
    Position Sampling=::NewEnumerator1
    Sampled Position Offset=&AddVectorToPosition(
     Position=$StackContext.ShapeLocation.ShapePosition
     Vector=&SelectVectorFromArray(
      Direct Array Index=$Particles.PartitionParticles.Partition
      Vector Selection Array=DataInterface() {propertyValues {FloatData [<0 0 30>, <0 -15 20>, <0 15 20>, <0 -15 20>, <0 15 20>]}}
     )
    )
    Sampled Position Offset Coordinate Space:ENiagaraSkelMeshTransforms=::NewEnumerator1
    Skeletal Mesh=$System.Skeletal Mesh
    Source Movement Scale:Vector3f=<0 0 0>
   ) {module=SkeletalMeshLocation001}
   &CurlNoiseForce(Pan Noise=true, Pan Noise Field=<0.05 0.1 0.2>) {module=CurlNoiseForce001}
   &SolveForcesAndVelocity(Clamp Velocity=false, Limit Acceleration=false)
   &ScaleColor(
    Scale Alpha=&FloatFromCurve(FloatCurve=DataInterface() {propertyValues {Curve {keys [{interpMode=RCIM_Cubic, tangentMode=RCTM_User, time=0, value=1}, {interpMode=RCIM_Cubic, tangentMode=RCTM_User, time=1, value=0}]}}})
    ScaleA=true
    ScaleRGB=true
   )
   &DynamicMaterialParameters(
    Index 0 Param 1=&FloatFromCurve(FloatCurve=DataInterface() {propertyValues {Curve {keys [
         {arriveTangent=-10.8792, interpMode=RCIM_Cubic, leaveTangent=-10.8792, tangentMode=RCTM_User, time=0, value=1}
         {tangentMode=RCTM_User, time=0.17, value=0}
        ]}}})
    Index 0 Param 2=&FloatFromCurve(FloatCurve=DataInterface() {propertyValues {Curve {keys [
         {arriveTangent=0.0202, interpMode=RCIM_Cubic, leaveTangent=0.0202, tangentMode=RCTM_User, time=0.98, value=0}
         {arriveTangent=229.7168, interpMode=RCIM_Cubic, leaveTangent=229.7168, tangentMode=RCTM_User, time=1, value=1.5}
        ]}}})
    Module.Param0WriteEnabled=true
    Module.Param1WriteEnabled=true
    Module.Param2WriteEnabled=true
    Module.Param3WriteEnabled=true
   )
  ]
  rendererClasses [&NiagaraSpriteRendererProperties]
  renderers [{rendererClass=&NiagaraSpriteRendererProperties}]
 }
}
summary {
 systemName=NS_Aura_Notes
 emitters [
  {emitterName=Aura, rendererClasses [&NiagaraSpriteRendererProperties]}
  {emitterName=Notes001, rendererClasses [&NiagaraSpriteRendererProperties]}
  {emitterName=Bubbles001, rendererClasses [&NiagaraSpriteRendererProperties]}
  {emitterName=Sparkles, rendererClasses [&NiagaraSpriteRendererProperties]}
 ]
 userVariables [
  User.Interpolate Intensity {type=NiagaraFloat, defaultValue=3, description=""}
  User.IsFrontEnd {type=NiagaraBool, defaultValue=true}
  User.IsLargeBody {type=NiagaraBool}
 ]
}
